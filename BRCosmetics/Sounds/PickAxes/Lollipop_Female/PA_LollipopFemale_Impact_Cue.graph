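# Sound cue for the impact of the female Lollipop pickaxe cosmetic.
# A Random node picks one of five equally weighted Circus Mallet enemy-impact waves as the root output.
# It plays through shared pickaxe impact attenuation, a cosmetic pickaxe audio property sheet and the impacts sound class, and is primed on load.
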SoundCue /BRCosmetics/Sounds/PickAxes/Lollipop_Female/PA_LollipopFemale_Impact_Cue
root=Random_0
nodes [
 {id=Random_0, children [WavePlayer_0,WavePlayer_1,WavePlayer_2,WavePlayer_3,WavePlayer_4], properties {weights [1,1,1,1,1]}}
 {id=WavePlayer_0, properties {soundWaveAssetPtr=&/Pickaxe_CircusMallet_EnemyImpact_01}}
 {id=WavePlayer_1, properties {soundWaveAssetPtr=&/Pickaxe_CircusMallet_EnemyImpact_02}}
 {id=WavePlayer_2, properties {soundWaveAssetPtr=&/Pickaxe_CircusMallet_EnemyImpact_04}}
 {id=WavePlayer_3, properties {soundWaveAssetPtr=&/Pickaxe_CircusMallet_EnemyImpact_05}}
 {id=WavePlayer_4, properties {soundWaveAssetPtr=&/Pickaxe_CircusMallet_EnemyImpact_06}}
]
properties {
 attenuationSettings=/Game/Sounds/Attenuation/Fort_Pickaxe_Impact_Att
 audioPropertiesSheet=/BRCosmetics/Sounds/Settings/PropertySheets/Pickaxe/APS_Cosmetic_Pickaxe_Impact
 bPrimeOnLoad=true
 maxDistance=4300
 soundClassObject=&Fort_Impacts
 volumeMultiplier=0.9
}
